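# Translucent, unlit, two-sided mesh-particle material for the Kurohomura backpack's orange, flame-like smoke effect.
# Emissive blends a UV gradient with distorted panning noise highlights, tinted by particle color.
# Opacity combines panning ink-smoke and soft trail textures with particle alpha and depth fade.
# Panned distortion normals drive vertex offset, with optional Rotate switch adding timed spin and dynamic-parameter push; mobile drops depth fade and distortion.
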Material /BRCosmetics/Accessories/FORT_Backpacks/M_MED_Kurohomura_Backpack/FX/M_MED_Kurhomura_Backpack
graph {
 properties {
  bUsedWithMeshParticles=true
  bUsedWithNiagaraMeshParticles=true
  blendMode=BLEND_Translucent
  pixelDepthOffsetMode=PDOM_Legacy
  shadingModel=MSM_Unlit
  twoSided=true
  usedShadingModels=Unlit
 }
 nodes [
  n1 = TextureCoordinate()
  n2 = LinearInterpolate(
   A=VectorParameter() {parameterName=Color, defaultValue=<rgba 1 0.2551 0 1>}.RGB
   B=VectorParameter() {parameterName=Color_Dark, defaultValue=<rgba 0.875 0.0506 0 1>}.RGB
   Alpha=Power(Base=ComponentMask(None=@n1) {g=true}) {constExponent=1.5}
  )
  n3 = VectorParameter() {parameterName=Noise_Color, defaultValue=<rgba 1 0.7157 0.1354 1>}
  n4 = TextureCoordinate() {coordinateIndex=1, uTiling=0.15, vTiling=0.5}
  n5 = TextureSample(UVs=Panner(Coordinate=@n4) {speedY=0.4}) {samplerType=SAMPLERTYPE_Normal, texture=&T_Weap_Distort_N_02}
  n6 = Add(A=@n4, B=Multiply(A=ComponentMask(None=@n5.RGB) {g=true, r=true}) {constB=0.12})
  n7 = Power(Base=TextureSample(UVs=Panner(Coordinate=@n6) {speedY=0.75}) {texture=/Game/Global/Textures/Default/Noise/TankMoteNoise}.R)
  n8 = TextureCoordinate() {coordinateIndex=1}
  n9 = Power(Base=OneMinus(None=ComponentMask(None=@n8) {g=true})) {constExponent=0.85}
  n10 = Clamp(None=Multiply(A=Multiply(A=@n7, B=@n9)) {constB=3.2}) {maxDefault=4.5}
  n11 = ParticleColor()
  n12 = Multiply(A=Multiply(A=LinearInterpolate(A=@n2, B=@n3.RGB, Alpha=@n10)) {constB=2.5}, B=@n11.RGB)
  n13 = MaterialFunctionCall(LDR Input=Clamp(None=@n12), HDR Tint=@n12) {materialFunction=&DeriveHDRfromLDR}
  MaterialFunctionCall(Emissive=ShadingPathSwitch(Default=@n13.RemappedRange, Deferred=@n13.RemappedRange, Mobile=@n12) {desc=Migrated from FeatureLevelSwitch}) {materialFunction=&MF_FX_EmissiveAdjustments}
  n14 = TextureSample(UVs=Panner(Coordinate=TextureCoordinate() {uTiling=0.8, vTiling=1.8}) {speedY=1.1}) {texture=&T_Inky_Smoke_Tile}
  n15 = Cosine(None=ComponentMask(None=@n8) {r=true})
  n16 = Clamp(None=Multiply(A=Subtract(A=@n14.R) {constB=0.1}, B=OneMinus(None=@n15)))
  n17 = TextureSample(UVs=MaterialFunctionCall(XY=TextureCoordinate() {uTiling=0.9, vTiling=1.5}) {materialFunction=&Swizzle}.YX) {samplerType=SAMPLERTYPE_Grayscale, texture=&T_Smoke_Trail_Soft}
  n18 = Clamp(None=Multiply(A=Add(A=@n16, B=@n17.R)) {constB=2.45})
  n19 = DepthFade(Opacity=@n18, FadeDistance=Constant() {r=10})
  Reroute(None=Multiply(A=@n11.A, B=ShadingPathSwitch(Default=@n19, Deferred=@n19, Mobile=@n18) {desc=Migrated from FeatureLevelSwitch}))
  n20 = TextureSample(UVs=Panner(Coordinate=TextureCoordinate() {coordinateIndex=1, uTiling=0.15, vTiling=0.8}) {speedY=0.6}) {samplerType=SAMPLERTYPE_Normal, texture=&T_Weap_Distort_N_02}
  n21 = AppendVector(A=@n20.R, B=Constant())
  n22 = TextureCoordinate()
  n23 = Multiply(
   A=Multiply(A=AppendVector(A=@n21, B=@n20.G), B=ScalarParameter() {defaultValue=2.4}) {constB=2.3}
   B=SphereMask(A=ComponentMask(None=@n22) {g=true}, B=Constant() {r=0.5}) {attenuationRadius=0.46, hardnessPercent=0}
  )
  n24 = ShadingPathSwitch(Default=@n23, Deferred=@n23, Mobile=Constant()) {desc=Migrated from FeatureLevelSwitch}
  n25 = Constant3Vector() {constant=<rgba 0 1 0 0>}
  n26 = RotateAboutAxis(
   NormalizedRotationAxis=Normalize(VectorInput=Transform(None=@n25) {transformSourceType=TRANSFORMSOURCE_Local})
   RotationAngle=Multiply(A=Constant() {r=-0.18}, B=Time())
   PivotPoint=ParticlePositionWS()
   Position=WorldPosition().XYZ
  )
  n27 = Constant3Vector() {constant=<rgba 0 1.5 0 0>}
  n28 = ComponentMask(None=@n27) {r=true}
  n29 = Multiply(A=ComponentMask(None=@n27) {g=true}, B=DynamicParameter().R)
  n30 = Transform(None=AppendVector(A=AppendVector(A=@n28, B=@n29), B=ComponentMask(None=@n27) {b=true})) {transformSourceType=TRANSFORMSOURCE_Local}
  n31 = SphereMask(A=WorldPosition().XYZ, B=ParticlePositionWS(), Radius=ScalarParameter() {defaultValue=10, parameterName=Param_1}) {attenuationRadius=0.5, hardnessPercent=0}
  StaticSwitchParameter(True=Add(A=Add(A=@n24, B=@n26), B=Multiply(A=@n30, B=@n31)), False=@n24) {parameterName=Rotate}
 ]
}
