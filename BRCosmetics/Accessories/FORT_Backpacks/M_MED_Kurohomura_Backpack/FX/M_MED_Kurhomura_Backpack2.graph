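# Translucent unlit, two-sided mesh-particle material for the Kurohomura backpack FX, drawing an animated orange radial glow.
# Radially mapped, panning caustics are added to a spinning radial mask and a scorch-hole texture distorted by a rotating portal normal map.
# A center-weighted sphere mask blends red-orange to hot yellow-orange, then particle color and emissive adjustments apply.
# Opacity combines particle alpha with a soft circular mask, depth-faded except on mobile.
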Material /BRCosmetics/Accessories/FORT_Backpacks/M_MED_Kurohomura_Backpack/FX/M_MED_Kurhomura_Backpack2
graph {
 properties {
  bUsedWithMeshParticles=true
  bUsedWithNiagaraMeshParticles=true
  blendMode=BLEND_Translucent
  pixelDepthOffsetMode=PDOM_Legacy
  shadingModel=MSM_Unlit
  twoSided=true
  usedShadingModels=Unlit
 }
 nodes [
  n1 = Multiply(
   A=MaterialFunctionCall() {materialFunction=&VectorToRadialValue}.Radial Coordinates
   B=AppendVector(A=Constant() {r=2}, B=Constant() {r=0.4})
  )
  n2 = Multiply(A=TextureSample(UVs=Panner(Coordinate=@n1) {speedY=1}) {texture=&T_ShieldCaustics}.R) {constB=2}
  n3 = Rotator() {speed=9}
  n4 = TextureSample(UVs=@n3) {samplerSource=SSM_Clamp_WorldGroupSettings, texture=/Game/Effects/Fort_Effects/Textures/Mask/T_RadialMask}
  n5 = TextureCoordinate()
  n6 = ComponentMask(None=TextureSample(UVs=@n3) {samplerType=SAMPLERTYPE_Normal, texture=/BRCosmetics/Effects/Fort_Effects/Textures/Taker/T_TakerPortal_Core_N}.RGB) {g=true, r=true}
  n7 = TextureSample(UVs=Add(A=Rotator(Coordinate=@n5) {speed=0}, B=Multiply(A=@n6) {constB=0.2})) {samplerSource=SSM_Clamp_WorldGroupSettings, texture=/Game/Effects/Fort_Effects/Gameplay/SmokeDustExplosions/Textures/T_Decal_Scorch_Hole}
  n8 = TextureCoordinate()
  n9 = Constant() {r=0.5}
  n10 = Multiply(
   A=Add(A=Add(A=@n2, B=@n4.G), B=Multiply(A=@n7.R) {constB=2})
   B=LinearInterpolate(
    A=Constant3Vector() {constant=<rgba 2 0.0984 0 0>}
    B=Constant3Vector() {constant=<rgba 12 6.9143 0.75 0>}
    Alpha=SphereMask(A=@n8, B=@n9) {attenuationRadius=0.32, hardnessPercent=0}
   )
  )
  n11 = ParticleColor()
  MaterialFunctionCall(Emissive=Multiply(A=Multiply(A=@n10) {constB=3}, B=@n11.RGB)) {materialFunction=&MF_FX_EmissiveAdjustments}
  n12 = SphereMask(A=TextureCoordinate(), B=Constant() {r=0.5}) {attenuationRadius=0.5, hardnessPercent=0}
  n13 = DepthFade(Opacity=@n12, FadeDistance=Constant() {r=20})
  Multiply(A=@n11.A, B=ShadingPathSwitch(Default=@n13, Deferred=@n13, Mobile=@n12) {desc=Migrated from FeatureLevelSwitch})
 ]
}
