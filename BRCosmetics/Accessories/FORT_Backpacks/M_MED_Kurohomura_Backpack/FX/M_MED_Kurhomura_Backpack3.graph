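# Unlit, translucent, two-sided effect material for the Kurohomura backpack cosmetic, usable on mesh particles, that draws animated orange flame-like energy.
# A wobble texture, panned on tiled, axis-swapped UVs, tints a bright orange color that feeds the shared emissive adjustment function.
# The same wobble channel is multiplied by a diagonally panning cartoon flame texture to form a moving, broken-up mask.
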Material /BRCosmetics/Accessories/FORT_Backpacks/M_MED_Kurohomura_Backpack/FX/M_MED_Kurhomura_Backpack3
graph {
 properties {
  bUsedWithMeshParticles=true
  bUsedWithNiagaraMeshParticles=true
  blendMode=BLEND_Translucent
  pixelDepthOffsetMode=PDOM_Legacy
  shadingModel=MSM_Unlit
  twoSided=true
  usedShadingModels=Unlit
 }
 nodes [
  n1 = TextureCoordinate() {uTiling=2}
  n2 = TextureSample(UVs=Panner(Coordinate=MaterialFunctionCall(XY=@n1) {materialFunction=&Swizzle}.YX) {speedY=-1.2}) {texture=/Game/Effects/Fort_Effects/Textures/T_Taker_Wobble}
  MaterialFunctionCall(Emissive=Multiply(A=Constant3Vector() {constant=<rgba 2.4 0.9266 0.1205 0>}, B=@n2.R)) {materialFunction=&MF_FX_EmissiveAdjustments}
  n3 = TextureCoordinate()
  Multiply(
   A=@n2.R
   B=TextureSample(UVs=Panner(Coordinate=@n3) {speedX=-0.7, speedY=1}) {texture=/Game/Effects/Fort_Effects/Textures/Tile/T_Tile_Texture_Cartoon_Flames}.R
  )
 ]
}
